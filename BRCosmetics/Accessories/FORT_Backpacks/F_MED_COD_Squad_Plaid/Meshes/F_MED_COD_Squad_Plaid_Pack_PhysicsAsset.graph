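# Physics asset for the F_MED_COD_Squad_Plaid backpack, defining simulation bodies and joints for its gun, canister, shells, ribbon and strap.
# Small box master bodies join each attachment to the main pack body with free swing and twist.
# Each part then hangs off its master with its own angular limits, tight for gun, canister and shells, loose for ribbon.
# The strap forms a three-link chain with wide swing limits.
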PhysicsAsset /BRCosmetics/Accessories/FORT_Backpacks/F_MED_COD_Squad_Plaid/Meshes/F_MED_COD_Squad_Plaid_Pack_PhysicsAsset
bodies {
 dyn_canister [{shapeType=Capsule, length=6.485, radius=4.0102, center=<-0.4126 0.1605 -0.877>, rotation=<rot 90 0 0>}]
 dyn_canister_mstr [{shapeType=Box, extentX=1.01, extentY=1.01, extentZ=1.01}]
 dyn_gun [{shapeType=Box, extentX=3.7676, extentY=17.9338, extentZ=11.5267, center=<2.6061 0 2.7073>, rotation=<rot 90 0 0>}]
 dyn_gun_mstr [{shapeType=Box, extentX=1.01, extentY=1.01, extentZ=1.01}]
 dyn_ribbon [{shapeType=Box, extentX=1.5, extentY=3.75, extentZ=7, center=<5 0.0714 -0.8799>, rotation=<rot 90 0 0>}]
 dyn_ribbon_mstr [{shapeType=Box, extentX=1.01, extentY=1.01, extentZ=1.01}]
 dyn_shells [{shapeType=Box, extentX=5.569, extentY=3.1211, extentZ=7.7784, center=<0.0717 -0.6243 -0.0279>}]
 dyn_shells_mstr [{shapeType=Box, extentX=1.01, extentY=1.01, extentZ=1.01}]
 dyn_strap_01 [{shapeType=Box, extentX=2.5, extentY=0.5, extentZ=4, center=<2.0649 0 0>, rotation=<rot 90 0 0>}]
 dyn_strap_02 [{shapeType=Box, extentX=2.25, extentY=0.5, extentZ=4.5, center=<1.2651 0 0>, rotation=<rot 90 0 0>}]
 dyn_strap_03 [{shapeType=Box, extentX=2, extentY=0.5, extentZ=3, center=<1.5 0 0>, rotation=<rot 90 0 0>}]
 dyn_strap_mstr [{shapeType=Box, extentX=1.01, extentY=1.01, extentZ=1.01}]
 pack [{shapeType=Box, extentX=32.5, extentY=7.5, extentZ=14.5, center=<5.2821 -4.0673 -1.5715>, rotation=<rot 44.9998 0 0>}]
}
constraints [
 {
  bone1Name=dyn_shells_mstr
  bone2Name=pack
  swing1LimitDegrees=45
  swing1Motion=Free
  swing2LimitDegrees=45
  swing2Motion=Free
  twistLimitDegrees=45
  twistMotion=Free
 }
 {
  bone1Name=dyn_gun_mstr
  bone2Name=pack
  swing1LimitDegrees=45
  swing1Motion=Free
  swing2LimitDegrees=45
  swing2Motion=Free
  twistLimitDegrees=45
  twistMotion=Free
 }
 {
  bone1Name=dyn_canister_mstr
  bone2Name=pack
  swing1LimitDegrees=45
  swing1Motion=Free
  swing2LimitDegrees=45
  swing2Motion=Free
  twistLimitDegrees=45
  twistMotion=Free
 }
 {
  bone1Name=dyn_ribbon_mstr
  bone2Name=pack
  swing1LimitDegrees=45
  swing1Motion=Free
  swing2LimitDegrees=45
  swing2Motion=Free
  twistLimitDegrees=45
  twistMotion=Free
 }
 {
  bone1Name=dyn_strap_mstr
  bone2Name=pack
  swing1LimitDegrees=45
  swing1Motion=Free
  swing2LimitDegrees=45
  swing2Motion=Free
  twistLimitDegrees=45
  twistMotion=Free
 }
 {bone1Name=dyn_gun, bone2Name=dyn_gun_mstr, swing1LimitDegrees=8, swing2LimitDegrees=8, twistLimitDegrees=5}
 {bone1Name=dyn_strap_01, bone2Name=dyn_strap_mstr, swing1LimitDegrees=75, swing2LimitDegrees=75, twistLimitDegrees=25}
 {bone1Name=dyn_strap_02, bone2Name=dyn_strap_01, swing1LimitDegrees=75, swing2LimitDegrees=75, twistLimitDegrees=25}
 {bone1Name=dyn_strap_03, bone2Name=dyn_strap_02, swing1LimitDegrees=55, swing2LimitDegrees=55, twistLimitDegrees=25}
 {bone1Name=dyn_canister, bone2Name=dyn_canister_mstr, swing1LimitDegrees=10, swing2LimitDegrees=15, twistLimitDegrees=15}
 {bone1Name=dyn_shells, bone2Name=dyn_shells_mstr, swing1LimitDegrees=10, swing2LimitDegrees=10, twistLimitDegrees=10}
 {bone1Name=dyn_ribbon, bone2Name=dyn_ribbon_mstr, swing1LimitDegrees=65, swing2LimitDegrees=65, twistLimitDegrees=25}
]
